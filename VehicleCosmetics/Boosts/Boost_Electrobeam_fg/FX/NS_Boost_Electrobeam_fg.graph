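# Niagara system for the Electrobeam vehicle boost, drawing rearward ribbon trails from skeletal mesh sockets, plus sprite emitters.
# Shared VC modules spawn ribbons per socket, then time- and age-driven sine offsets wave them, varying by spawn group.
# Some emitters also write CustomPosition and secondary widths and have multiple ribbon renderers; several ribbon variants are disabled.
# User.bIsPainted switches colors and ribbon alpha between default pink and painted values.
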NiagaraSystem /VehicleCosmetics/Boosts/Boost_Electrobeam_fg/FX/NS_Boost_Electrobeam_fg
dependencies {
 usedDataInterfaces [
  &NiagaraDataInterfaceArrayFloat
  &NiagaraDataInterfaceCurve
  &NiagaraDataInterfaceVectorCurve
  &NiagaraDataInterfaceSkeletalMesh
 ]
 usedDynamicInputs []
 usedModules [
  &self:SystemSpawnScript.SystemScriptSource.SystemScriptGraph.NiagaraNodeAssignment_0.SetVariables_#1
  &SystemState
 ]
}
emitters {
 NE_VC_Boost_Ribbon_Parent {
  bEnabled=false
  emitterSpawnScript [/SetVariables(
    Emitter.IsPainted:NiagaraFloat=&MakeCustomFloatFromBool(Bool=$User.bIsPainted, True Float=2, False Float=-0.512)
    Emitter.MirrorOffset:NiagaraFloat=&MakeCustomFloatFromBool(Bool=$User.bIsPainted, True Float=4, False Float=2)
   ) {isSetParameters=true, module=SetVariables_#2}]
  emitterUpdateScript [
   &EmitterState(Life Cycle Mode=::NewEnumerator1, Loop Duration=1, UseLoopDelay=false)
   &VC_FilteredSocketCount()
   &VC_SpawnRibbon_v1(Number of Ribbons=$Emitter.PortCount, SpawnRate=120)
   /SetVariables(
    Emitter.AnotherOffset:NiagaraFloat=&Power(a=HlslExpression() {hlslExpression="(fmod(Engine.Time * 0.35, 0.2) + 0.0) * 3"}, b=1)
    Emitter.TimeLerp:NiagaraFloat=&Power(a=HlslExpression() {hlslExpression="(fmod(Engine.Time * 0.325, 0.4) + 0.15) * 3"}, b=2)
    Emitter.TimeOffset:NiagaraFloat=&Power(a=HlslExpression() {hlslExpression="fmod(Engine.Time * 0.5, 0.3) + 0.15 * 3"}, b=2)
   ) {isSetParameters=true, module=SetVariables_#3}
   /SetVariables(
    Emitter.LerpChannel:Vector3f=&Lerp_Vector(
     A=<1 0 0>
     B=<0 1 0>
     Alpha=&Ceil_Float(Float=&Sine(Normalized Angle=$Emitter.Age, Period=$Emitter.TimeLerp, Scale=1, Bias=0))
    )
    Emitter.OffsetX:NiagaraFloat=&Multiply_Float(
     A=&Ceil_Float(Float=&Sine(Normalized Angle=$Emitter.Age, Period=$Emitter.TimeOffset, Scale=1, Bias=0))
     B=0.25
    )
   ) {isSetParameters=true, module=SetVariables_#4}
  ]
  particleSpawnScript [
   &VC_PaintedVariantsOverride() {enabled=false}
   /SetVariables(
    Particles.ColorTwo:LinearColor=<rgba 6 0 5.2125 1>
    Particles.SizeTwo:NiagaraFloat=&RandomRangeFloat(Minimum=20, Maximum=30)
   ) {isSetParameters=true, module=SetVariables_#5}
   &VC_InitializeRibbon_v1(
    Color=&MakeCustomLinearColorFromBool(Bool=$User.bIsPainted, True Color=$User.BelowTrailColor, False Color=<rgba 24 3.0925 20.1246 1>)
    Custom Offset=true
    Enable Skeletal Position=true
    Lifetime=0.46
    Offset Amount=1
    Ribbon Width:NiagaraFloat=&RandomRangeFloat(Minimum=90, Maximum=90)
   )
   &AddVelocity(Velocity=<-2500 0 50>) {module=AddVelocity001}
   &ShapeLocation(MaskSpawnGroup=false, Override Seed=false, Sphere Radius=7, UseOwnerScale=true) {enabled=false}
   /SetVariables(Particles.MainRibbonWidth:NiagaraFloat=&MaskFloat(Mask Source=0, Mask Parameter=?, Mask=0, Masked Float=80, Passthrough Float=-80)) {isSetParameters=true, module=SetVariables_#6}
  ]
  particleUpdateScript [
   &ParticleState()
   &ScaleColor(
    Scale Alpha=&MakeCustomFloatFromBool(
     Bool=$User.bIsPainted
     True Float=0.15
     False Float=&FloatFromCurve(FloatCurve=DataInterface() {propertyValues {Curve {keys [<at 0 0>,<at 0.02 0.1>,<at 0.95 0.1>,<at 1 0>]}}}, Scale Curve=0.2)
    )
    ScaleA=true
    ScaleRGB=true
   )
   &ScaleRibbonWidth(Ribbon Width Scale=&FloatFromCurve(
     FloatCurve=DataInterface() {propertyValues {Curve {keys [
         {interpMode=RCIM_Cubic, tangentMode=RCTM_User, time=0, value=0}
         {arriveTangent=0.5028, interpMode=RCIM_Cubic, leaveTangent=0.5028, tangentMode=RCTM_User, time=0.1, value=1.25}
         {arriveTangent=-0.5446, interpMode=RCIM_Cubic, leaveTangent=-0.5446, tangentMode=RCTM_User, time=0.8, value=1}
         {arriveTangent=-3.985, interpMode=RCIM_Cubic, leaveTangent=-3.985, tangentMode=RCTM_User, time=0.9, value=0}
        ]}}}
     Scale Curve=1.25
    ))
   /SetVariables(
    Particles.CustomWidth:NiagaraFloat=&Multiply_Float(
     A=$Particles.MainRibbonWidth
     B=&FloatFromCurve(
      FloatCurve=DataInterface() {propertyValues {Curve {keys [
          {interpMode=RCIM_Cubic, tangentMode=RCTM_User, time=0, value=0}
          {arriveTangent=0.5028, interpMode=RCIM_Cubic, leaveTangent=0.5028, tangentMode=RCTM_User, time=0.1, value=1.25}
          {arriveTangent=-0.5446, interpMode=RCIM_Cubic, leaveTangent=-0.5446, tangentMode=RCTM_User, time=0.8, value=1}
          {arriveTangent=-3.985, interpMode=RCIM_Cubic, leaveTangent=-3.985, tangentMode=RCTM_User, time=0.9, value=0}
         ]}}}
      Scale Curve=1.25
     )
    )
    Particles.SizeScaleTwo:NiagaraFloat=&Multiply_Float(
     A=$Particles.SizeTwo
     B=&FloatFromCurve(FloatCurve=DataInterface() {propertyValues {Curve {keys [
          {arriveTangent=-0.0009, time=0, value=0}
          <at 0.075 0.75 RCIM_Cubic>
          {arriveTangent=-1, leaveTangent=-1, time=0.175, value=0}
          <at 0.65 0>
          <at 0.825 0 RCIM_Cubic>
          {interpMode=RCIM_Cubic, leaveTangent=-2.5286, time=1, value=0}
         ]}}})
    )
   ) {isSetParameters=true, module=SetVariables_#7}
   &SolveForcesAndVelocity(Clamp Velocity=false, Limit Acceleration=false)
   &VC_UpdateRibbon_v1(Enable Skeletal Position=true)
   /SetVariables(Particles.Position=&AddVectorToPosition(
     Position=$Particles.Position
     Vector=&TransformVector(
      Vector=&MakeVector(
       X=0
       Y=0
       Z=&Sine(
        Normalized Angle=&Multiply_Float(
         A=1
         B=&Subtract_Float(
          A=&Multiply_Float(A=2, B=&Subtract_Float(A=$Engine.Time, B=0.5))
          B=$Particles.NormalizedAge
         )
        )
        Period=1
        Scale=&FloatFromCurve(FloatCurve=DataInterface() {propertyValues {Curve {keys [<at 0.5 6 RCIM_Cubic>,<at 1 8 RCIM_Cubic>]}}})
        Bias=0
       )
      )
      Destination Space=::Simulation
      Apply Scale=true
     )
    )) {enabled=false, isSetParameters=true, module=SetVariables_#8}
   /SetVariables(Particles.Position=&AddVectorToPosition(
     Position=$Particles.Position
     Vector=&TransformVector(
      Vector=&MakeVector(
       X=0
       Y=&Sine(
        Normalized Angle=&Multiply_Float(
         A=&SelectFloatFromArray(Direct Array Index=&SpawnGroup(), Float Selection Array=DataInterface() {propertyValues {FloatData [-1,1]}})
         B=&Subtract_Float(
          A=&Multiply_Float(A=2, B=&Subtract_Float(A=$Engine.Time, B=0.5))
          B=$Particles.NormalizedAge
         )
        )
        Period=1
        Scale=&FloatFromCurve(FloatCurve=DataInterface() {propertyValues {Curve {keys [<at 0.5 8 RCIM_Cubic>,<at 1 6 RCIM_Cubic>]}}})
        Bias=0
       )
       Z=&Sine(
        Normalized Angle=&Multiply_Float(
         A=1
         B=&Subtract_Float(
          A=&Multiply_Float(A=2, B=&Subtract_Float(A=$Engine.Time, B=0.5))
          B=$Particles.NormalizedAge
         )
        )
        Period=1
        Scale=&FloatFromCurve(FloatCurve=DataInterface() {propertyValues {Curve {keys [<at 0.25 4 RCIM_Cubic>,<at 0.5 0 RCIM_Cubic>]}}})
        Bias=0
       )
      )
      Destination Space=::Simulation
      Apply Scale=true
     )
    )) {isSetParameters=true, module=SetVariables_#9}
   /SetVariables(Particles.CustomPosition:NiagaraPosition=&AddVectorToPosition(
     Position=$Particles.Position
     Vector=&TransformVector(
      Vector=&MakeVector(
       X=0
       Y=&Sine(
        Normalized Angle=&Multiply_Float(
         A=&SelectFloatFromArray(Direct Array Index=&SpawnGroup(), Float Selection Array=DataInterface() {propertyValues {FloatData [-1,1]}})
         B=&Subtract_Float(
          A=&Multiply_Float(A=2, B=&Subtract_Float(A=$Engine.Time, B=0.5))
          B=$Particles.NormalizedAge
         )
        )
        Period=0.25
        Scale=&FloatFromCurve(FloatCurve=DataInterface() {propertyValues {Curve {keys [<at 0.5 4 RCIM_Cubic>,<at 1 0 RCIM_Cubic>]}}})
        Bias=0
       )
       Z=&Sine(
        Normalized Angle=&Multiply_Float(
         A=1
         B=&Subtract_Float(
          A=&Multiply_Float(A=2, B=&Subtract_Float(A=$Engine.Time, B=0.5))
          B=$Particles.NormalizedAge
         )
        )
        Period=0.25
        Scale=&FloatFromCurve(FloatCurve=DataInterface() {propertyValues {Curve {keys [<at 0.5 4 RCIM_Cubic>,<at 1 0 RCIM_Cubic>]}}})
        Bias=0
       )
      )
      Destination Space=::Simulation
      Apply Scale=true
     )
    )) {isSetParameters=true, module=SetVariables_#10}
  ]
  rendererClasses [&NiagaraRibbonRendererProperties]
  renderers [
   {rendererClass=&NiagaraRibbonRendererProperties}
   {rendererClass=&NiagaraRibbonRendererProperties, rendererIndex=1}
  ]
 }
 NE_VC_Boost_Ribbon_Parent001 {
  emitterUpdateScript [
   &EmitterState()
   &VC_FilteredSocketCount()
   &VC_RateQualityHandler(High Spawn Rate=60, Low Spawn Rate=60, Medium Spawn Rate=60, Use Custom Rate?=true)
   &VC_SpawnRibbon_v1(Number of Ribbons=$Emitter.PortCount, SpawnRate=$Emitter.Spawn Rate)
   /SetVariables(Emitter.TimeOffset:NiagaraFloat=&Power(a=HlslExpression() {hlslExpression="(fmod(Engine.DeltaTime * 0.35, 0.2) + 0.0) * 3"}, b=2)) {isSetParameters=true, module=SetVariables_#11}
  ]
  particleSpawnScript [
   &VC_PaintedVariantsOverride() {enabled=false}
   &VC_InitializeRibbon_v1(
    Custom Offset=true
    Enable Skeletal Position=true
    Lifetime=0.7
    Local Vector Offset=<20 0 0>
    Offset Amount=1
    Ribbon Width:NiagaraFloat=&RandomRangeFloat(Minimum=64, Maximum=64)
   )
   &AddVelocity(Velocity=<-2000 0 0>) {module=AddVelocity001}
   &ShapeLocation(MaskSpawnGroup=false, Override Seed=false, Sphere Radius=2, UseOwnerScale=true) {enabled=false}
  ]
  particleUpdateScript [
   &ParticleState()
   &ScaleRibbonWidth(Ribbon Width Scale=&FloatFromCurve(FloatCurve=DataInterface() {propertyValues {Curve {keys [
         {arriveTangent=1.0826, interpMode=RCIM_Cubic, leaveTangent=1.0826, tangentMode=RCTM_User, time=0, value=0.5}
         {arriveTangent=0.0088, interpMode=RCIM_Cubic, leaveTangent=0.0088, tangentMode=RCTM_User, time=0.25, value=0.7}
         {arriveTangent=-1.7516, interpMode=RCIM_Cubic, leaveTangent=-1.7516, tangentMode=RCTM_User, time=0.8, value=0}
        ]}}}))
   &SolveForcesAndVelocity(Clamp Velocity=false, Limit Acceleration=false)
   &VC_UpdateRibbon_v1(Enable Skeletal Position=true)
   /SetVariables(Particles.Position=&AddVectorToPosition(
     Position=$Particles.Position
     Vector=&TransformVector(
      Vector=&MakeVector(
       X=0
       Y=&Sine(
        Normalized Angle=&Multiply_Float(
         A=1
         B=&Subtract_Float(
          A=&Multiply_Float(A=2, B=&Subtract_Float(A=$Engine.Time, B=0.5))
          B=$Particles.NormalizedAge
         )
        )
        Period=&SelectFloatFromArray(
         Direct Array Index=&SpawnGroup()
         Float Selection Array=DataInterface() {propertyValues {FloatData [0.7,0.765]}}
        )
        Scale=&FloatFromCurve(FloatCurve=DataInterface() {propertyValues {Curve {keys [<at 0.1 2.5 RCIM_Cubic>,<at 0.86 0 RCIM_Cubic>]}}})
        Bias=0
       )
       Z=&Sine(
        Normalized Angle=&Multiply_Float(
         A=1
         B=&Subtract_Float(
          A=&Multiply_Float(A=2, B=&Subtract_Float(A=$Engine.Time, B=0.5))
          B=$Particles.NormalizedAge
         )
        )
        Period=&SelectFloatFromArray(
         Direct Array Index=&SpawnGroup()
         Float Selection Array=DataInterface() {propertyValues {FloatData [0.3,0.365]}}
        )
        Scale=&FloatFromCurve(FloatCurve=DataInterface() {propertyValues {Curve {keys [<at 0.1 3 RCIM_Cubic>,<at 0.86 0 RCIM_Cubic>]}}})
        Bias=0
       )
      )
      Destination Space=::Simulation
      Apply Scale=true
     )
    )) {isSetParameters=true, module=SetVariables_#12}
  ]
  rendererClasses [&NiagaraRibbonRendererProperties]
  renderers [{rendererClass=&NiagaraRibbonRendererProperties}]
 }
 NE_VC_Boost_Ribbon_Parent002 {
  bEnabled=false
  emitterUpdateScript [
   &EmitterState()
   &VC_FilteredSocketCount()
   &VC_RateQualityHandler(High Spawn Rate=60, Low Spawn Rate=60, Medium Spawn Rate=60, Use Custom Rate?=true)
   &VC_SpawnRibbon_v1(Number of Ribbons=$Emitter.PortCount, SpawnRate=$Emitter.Spawn Rate)
   /SetVariables(Emitter.TimeOffset:NiagaraFloat=&Power(a=HlslExpression() {hlslExpression="(fmod(Engine.DeltaTime * 0.35, 0.2) + 0.0) * 3"}, b=2)) {isSetParameters=true, module=SetVariables_#11}
  ]
  particleSpawnScript [
   &VC_PaintedVariantsOverride() {enabled=false}
   &VC_InitializeRibbon_v1(
    Custom Offset=true
    Enable Skeletal Position=true
    Lifetime=0.7
    Local Vector Offset=<20 0 0>
    Offset Amount=1
    Ribbon Width:NiagaraFloat=&RandomRangeFloat(Minimum=64, Maximum=64)
   )
   &AddVelocity(Velocity=<-2000 0 0>) {module=AddVelocity001}
   &ShapeLocation(MaskSpawnGroup=false, Override Seed=false, Sphere Radius=2, UseOwnerScale=true) {enabled=false}
  ]
  particleUpdateScript [
   &ParticleState()
   &ScaleRibbonWidth(Ribbon Width Scale=&FloatFromCurve(FloatCurve=DataInterface() {propertyValues {Curve {keys [
         {arriveTangent=1.0826, interpMode=RCIM_Cubic, leaveTangent=1.0826, tangentMode=RCTM_User, time=0, value=0.5}
         {arriveTangent=0.0088, interpMode=RCIM_Cubic, leaveTangent=0.0088, tangentMode=RCTM_User, time=0.25, value=0.7}
         {arriveTangent=-1.7516, interpMode=RCIM_Cubic, leaveTangent=-1.7516, tangentMode=RCTM_User, time=0.8, value=0}
        ]}}}))
   &SolveForcesAndVelocity(Clamp Velocity=false, Limit Acceleration=false)
   &VC_UpdateRibbon_v1(Enable Skeletal Position=true)
   /SetVariables(Particles.Position=&AddVectorToPosition(
     Position=$Particles.Position
     Vector=&TransformVector(
      Vector=&MakeVector(
       X=0
       Y=&Sine(
        Normalized Angle=&Multiply_Float(
         A=&SelectFloatFromArray(Direct Array Index=&SpawnGroup(), Float Selection Array=DataInterface() {propertyValues {FloatData [-1,1]}})
         B=&Subtract_Float(
          A=&Multiply_Float(A=2, B=&Subtract_Float(A=$Engine.Time, B=0.5))
          B=$Particles.NormalizedAge
         )
        )
        Period=1
        Scale=&FloatFromCurve(FloatCurve=DataInterface() {propertyValues {Curve {keys [<at 0 2 RCIM_Cubic>,<at 1 0 RCIM_Cubic>]}}})
        Bias=0
       )
       Z=&Sine(
        Normalized Angle=&Add_Float(
         A=&Multiply_Float(
          A=$Particles.Age
          B=&SelectFloatFromArray(Direct Array Index=$Particles.SpawnGroup, Float Selection Array=DataInterface() {propertyValues {FloatData [-1,1]}})
         )
         B=&Multiply_Float(A=$Emitter.LoopedAge, B=0)
        )
        Period=$Emitter.TimeOffset
        Scale=&FloatFromCurve(
         FloatCurve=DataInterface() {propertyValues {Curve {keys [{arriveTangent=0.7811, interpMode=RCIM_Cubic, leaveTangent=0.7811, tangentMode=RCTM_User, time=0, value=0.85}, <at 0.5 1>]}}}
         Scale Curve=5
        )
        Bias=0
       )
      )
      Destination Space=::Simulation
      Apply Scale=true
     )
    )) {isSetParameters=true, module=SetVariables_#12}
   /SetVariables(Particles.Position=&AddVectorToPosition(
     Position=$Particles.Position
     Vector=&TransformVector(
      Vector=&MakeVector(
       X=0
       Y=&Sine(
        Normalized Angle=&Multiply_Float(
         A=&SelectFloatFromArray(Direct Array Index=&SpawnGroup(), Float Selection Array=DataInterface() {propertyValues {FloatData [-1,1]}})
         B=&Subtract_Float(
          A=&Multiply_Float(A=2, B=&Subtract_Float(A=$Engine.Time, B=0.5))
          B=$Particles.NormalizedAge
         )
        )
        Period=1
        Scale=&FloatFromCurve(FloatCurve=DataInterface() {propertyValues {Curve {keys [<at 0 2 RCIM_Cubic>,<at 1 0 RCIM_Cubic>]}}})
        Bias=0
       )
       Z=&Sine(
        Normalized Angle=&Multiply_Float(
         A=&SelectFloatFromArray(Direct Array Index=&SpawnGroup(), Float Selection Array=DataInterface() {propertyValues {FloatData [-1,1]}})
         B=&Subtract_Float(
          A=&Multiply_Float(A=2, B=&Subtract_Float(A=$Engine.Time, B=0.5))
          B=$Particles.NormalizedAge
         )
        )
        Period=0.4
        Scale=&FloatFromCurve(FloatCurve=DataInterface() {propertyValues {Curve {keys [<at 0 4 RCIM_Cubic>,<at 1 0 RCIM_Cubic>]}}})
        Bias=0
       )
      )
      Destination Space=::Simulation
      Apply Scale=true
     )
    )) {enabled=false, isSetParameters=true}
  ]
  rendererClasses [&NiagaraRibbonRendererProperties]
  renderers [{rendererClass=&NiagaraRibbonRendererProperties}]
 }
 NE_VC_Boost_Ribbon_Parent003 {
  emitterSpawnScript [/SetVariables(
    Emitter.IsPainted:NiagaraFloat=&MakeCustomFloatFromBool(Bool=$User.bIsPainted, True Float=2, False Float=-0.512)
    Emitter.MirrorOffset:NiagaraFloat=&MakeCustomFloatFromBool(Bool=$User.bIsPainted, True Float=4, False Float=2)
   ) {isSetParameters=true, module=SetVariables_#2}]
  emitterUpdateScript [
   &EmitterState(Life Cycle Mode=::NewEnumerator1, Loop Duration=1, UseLoopDelay=false)
   &VC_FilteredSocketCount()
   &VC_SpawnRibbon_v1(Number of Ribbons=$Emitter.PortCount, SpawnRate=120)
   /SetVariables(
    Emitter.AnotherOffset:NiagaraFloat=&Power(a=HlslExpression() {hlslExpression="(fmod(Engine.Time * 0.35, 0.2) + 0.0) * 3"}, b=1)
    Emitter.TimeLerp:NiagaraFloat=&Power(a=HlslExpression() {hlslExpression="(fmod(Engine.Time * 0.325, 0.4) + 0.15) * 3"}, b=2)
    Emitter.TimeOffset:NiagaraFloat=&Power(a=HlslExpression() {hlslExpression="fmod(Engine.Time * 0.5, 0.3) + 0.15 * 3"}, b=2)
   ) {isSetParameters=true, module=SetVariables_#3}
   /SetVariables(
    Emitter.LerpChannel:Vector3f=&Lerp_Vector(
     A=<1 0 0>
     B=<0 1 0>
     Alpha=&Ceil_Float(Float=&Sine(Normalized Angle=$Emitter.Age, Period=$Emitter.TimeLerp, Scale=1, Bias=0))
    )
    Emitter.OffsetX:NiagaraFloat=&Multiply_Float(
     A=&Ceil_Float(Float=&Sine(Normalized Angle=$Emitter.Age, Period=$Emitter.TimeOffset, Scale=1, Bias=0))
     B=0.25
    )
   ) {isSetParameters=true, module=SetVariables_#4}
  ]
  particleSpawnScript [
   &VC_PaintedVariantsOverride() {enabled=false}
   /SetVariables(
    Particles.ColorTwo:LinearColor=<rgba 6 0 5.2125 1>
    Particles.SizeTwo:NiagaraFloat=&RandomRangeFloat(Minimum=20, Maximum=30)
   ) {isSetParameters=true, module=SetVariables_#5}
   &VC_InitializeRibbon_v1(
    Color=&MakeCustomLinearColorFromBool(Bool=$User.bIsPainted, True Color=$User.BelowTrailColor, False Color=<rgba 24 3.0925 20.1246 1>)
    Custom Offset=true
    Enable Skeletal Position=true
    Lifetime=0.46
    Offset Amount=1
    Ribbon Width:NiagaraFloat=&RandomRangeFloat(Minimum=90, Maximum=90)
   )
   &AddVelocity(Velocity=<-2500 0 50>) {module=AddVelocity001}
   &ShapeLocation(MaskSpawnGroup=false, Override Seed=false, Sphere Radius=7, UseOwnerScale=true) {enabled=false}
   /SetVariables(Particles.MainRibbonWidth:NiagaraFloat=&MaskFloat(Mask Source=0, Mask Parameter=?, Mask=0, Masked Float=80, Passthrough Float=-80)) {isSetParameters=true, module=SetVariables_#6}
  ]
  particleUpdateScript [
   &ParticleState()
   &ScaleColor(
    Scale Alpha=&MakeCustomFloatFromBool(
     Bool=$User.bIsPainted
     True Float=0.15
     False Float=&FloatFromCurve(FloatCurve=DataInterface() {propertyValues {Curve {keys [<at 0 0>,<at 0.02 0.1>,<at 0.95 0.1>,<at 1 0>]}}}, Scale Curve=0.2)
    )
    ScaleA=true
    ScaleRGB=true
   )
   &ScaleRibbonWidth(Ribbon Width Scale=&FloatFromCurve(
     FloatCurve=DataInterface() {propertyValues {Curve {keys [
         {interpMode=RCIM_Cubic, tangentMode=RCTM_User, time=0, value=0}
         {arriveTangent=0.5028, interpMode=RCIM_Cubic, leaveTangent=0.5028, tangentMode=RCTM_User, time=0.1, value=1.25}
         {arriveTangent=-0.5446, interpMode=RCIM_Cubic, leaveTangent=-0.5446, tangentMode=RCTM_User, time=0.8, value=1}
         {arriveTangent=-3.985, interpMode=RCIM_Cubic, leaveTangent=-3.985, tangentMode=RCTM_User, time=0.9, value=0}
        ]}}}
     Scale Curve=1.25
    ))
   /SetVariables(
    Particles.CustomWidth:NiagaraFloat=&Multiply_Float(
     A=$Particles.MainRibbonWidth
     B=&FloatFromCurve(
      FloatCurve=DataInterface() {propertyValues {Curve {keys [
          {interpMode=RCIM_Cubic, tangentMode=RCTM_User, time=0, value=0}
          {arriveTangent=0.5028, interpMode=RCIM_Cubic, leaveTangent=0.5028, tangentMode=RCTM_User, time=0.1, value=1.25}
          {arriveTangent=-0.5446, interpMode=RCIM_Cubic, leaveTangent=-0.5446, tangentMode=RCTM_User, time=0.8, value=1}
          {arriveTangent=-3.985, interpMode=RCIM_Cubic, leaveTangent=-3.985, tangentMode=RCTM_User, time=0.9, value=0}
         ]}}}
      Scale Curve=1.25
     )
    )
    Particles.SizeScaleTwo:NiagaraFloat=&Multiply_Float(
     A=$Particles.SizeTwo
     B=&FloatFromCurve(FloatCurve=DataInterface() {propertyValues {Curve {defaultValue=0, keys [<at 0 0>,<at 0.075 0.75 RCIM_Cubic>,<at 0.175 0>,<at 0.65 0>,<at 0.825 0 RCIM_Cubic>,<at 1 0 RCIM_Cubic>]}}})
    )
   ) {isSetParameters=true, module=SetVariables_#7}
   &SolveForcesAndVelocity(Clamp Velocity=false, Limit Acceleration=false)
   &VC_UpdateRibbon_v1(Enable Skeletal Position=true)
   /SetVariables(Particles.Position=&AddVectorToPosition(
     Position=$Particles.Position
     Vector=&TransformVector(
      Vector=&MakeVector(
       X=0
       Y=&Sine(
        Normalized Angle=&Multiply_Float(
         A=1
         B=&Subtract_Float(
          A=&Multiply_Float(A=2, B=&Subtract_Float(A=$Engine.Time, B=0.5))
          B=$Particles.NormalizedAge
         )
        )
        Period=&SelectFloatFromArray(
         Direct Array Index=&SpawnGroup()
         Float Selection Array=DataInterface() {propertyValues {FloatData [0.7,0.765]}}
        )
        Scale=&FloatFromCurve(FloatCurve=DataInterface() {propertyValues {Curve {keys [<at 0 10 RCIM_Cubic>,<at 0.86 0 RCIM_Cubic>]}}})
        Bias=0
       )
       Z=&Sine(
        Normalized Angle=&Multiply_Float(
         A=1
         B=&Subtract_Float(
          A=&Multiply_Float(A=2, B=&Subtract_Float(A=$Engine.Time, B=0.5))
          B=$Particles.NormalizedAge
         )
        )
        Period=1
        Scale=&FloatFromCurve(FloatCurve=DataInterface() {propertyValues {Curve {keys [<at 0.25 4 RCIM_Cubic>,<at 0.5 0 RCIM_Cubic>]}}})
        Bias=0
       )
      )
      Destination Space=::Simulation
      Apply Scale=true
     )
    )) {isSetParameters=true, module=SetVariables_#9}
   /SetVariables(Particles.CustomPosition:NiagaraPosition=&AddVectorToPosition(
     Position=$Particles.Position
     Vector=&TransformVector(
      Vector=&MakeVector(
       X=0
       Y=&Sine(
        Normalized Angle=&Multiply_Float(
         A=&SelectFloatFromArray(Direct Array Index=&SpawnGroup(), Float Selection Array=DataInterface() {propertyValues {FloatData [-1,1]}})
         B=&Subtract_Float(
          A=&Multiply_Float(A=2, B=&Subtract_Float(A=$Engine.Time, B=0.5))
          B=$Particles.NormalizedAge
         )
        )
        Period=0.25
        Scale=&FloatFromCurve(FloatCurve=DataInterface() {propertyValues {Curve {keys [<at 0.5 4 RCIM_Cubic>,<at 1 0 RCIM_Cubic>]}}})
        Bias=0
       )
       Z=&Sine(
        Normalized Angle=&Multiply_Float(
         A=1
         B=&Subtract_Float(
          A=&Multiply_Float(A=2, B=&Subtract_Float(A=$Engine.Time, B=0.5))
          B=$Particles.NormalizedAge
         )
        )
        Period=0.25
        Scale=&FloatFromCurve(FloatCurve=DataInterface() {propertyValues {Curve {keys [<at 0.5 4 RCIM_Cubic>,<at 1 0 RCIM_Cubic>]}}})
        Bias=0
       )
      )
      Destination Space=::Simulation
      Apply Scale=true
     )
    )) {isSetParameters=true, module=SetVariables_#10}
  ]
  rendererClasses [&NiagaraRibbonRendererProperties]
  renderers [
   {rendererClass=&NiagaraRibbonRendererProperties}
   {rendererClass=&NiagaraRibbonRendererProperties, rendererIndex=1}
  ]
 }
 NE_VC_Boost_Ribbon_Parent004 {
  bEnabled=false
  emitterSpawnScript [/SetVariables(
    Emitter.IsPainted:NiagaraFloat=&MakeCustomFloatFromBool(Bool=$User.bIsPainted, True Float=2, False Float=-0.512)
    Emitter.MirrorOffset:NiagaraFloat=&MakeCustomFloatFromBool(Bool=$User.bIsPainted, True Float=4, False Float=2)
   ) {isSetParameters=true, module=SetVariables_#2}]
  emitterUpdateScript [
   &EmitterState(Life Cycle Mode=::NewEnumerator1, Loop Duration=1, UseLoopDelay=false)
   &VC_FilteredSocketCount()
   &VC_SpawnRibbon_v1(Number of Ribbons=$Emitter.PortCount, SpawnRate=120)
   /SetVariables(
    Emitter.AnotherOffset:NiagaraFloat=&Power(a=HlslExpression() {hlslExpression="(fmod(Engine.Time * 0.35, 0.2) + 0.0) * 3"}, b=1)
    Emitter.TimeLerp:NiagaraFloat=&Power(a=HlslExpression() {hlslExpression="(fmod(Engine.Time * 0.325, 0.4) + 0.15) * 3"}, b=2)
    Emitter.TimeOffset:NiagaraFloat=&Power(a=HlslExpression() {hlslExpression="fmod(Engine.Time * 0.5, 0.3) + 0.15 * 3"}, b=2)
   ) {isSetParameters=true, module=SetVariables_#3}
   /SetVariables(
    Emitter.LerpChannel:Vector3f=&Lerp_Vector(
     A=<1 0 0>
     B=<0 1 0>
     Alpha=&Ceil_Float(Float=&Sine(Normalized Angle=$Emitter.Age, Period=$Emitter.TimeLerp, Scale=1, Bias=0))
    )
    Emitter.OffsetX:NiagaraFloat=&Multiply_Float(
     A=&Ceil_Float(Float=&Sine(Normalized Angle=$Emitter.Age, Period=$Emitter.TimeOffset, Scale=1, Bias=0))
     B=0.25
    )
   ) {isSetParameters=true, module=SetVariables_#4}
  ]
  particleSpawnScript [
   &VC_PaintedVariantsOverride() {enabled=false}
   /SetVariables(
    Particles.ColorTwo:LinearColor=<rgba 6 0 5.2125 1>
    Particles.SizeTwo:NiagaraFloat=&RandomRangeFloat(Minimum=20, Maximum=30)
   ) {isSetParameters=true, module=SetVariables_#5}
   &VC_InitializeRibbon_v1(
    Color=&MakeCustomLinearColorFromBool(Bool=$User.bIsPainted, True Color=$User.BelowTrailColor, False Color=<rgba 24 3.0925 20.1246 1>)
    Custom Offset=true
    Enable Skeletal Position=true
    Lifetime=0.46
    Offset Amount=1
    Ribbon Width:NiagaraFloat=&RandomRangeFloat(Minimum=90, Maximum=90)
   )
   &AddVelocity(Velocity=<-2500 0 50>) {module=AddVelocity001}
   &ShapeLocation(MaskSpawnGroup=false, Override Seed=false, Sphere Radius=7, UseOwnerScale=true) {enabled=false}
   /SetVariables(Particles.MainRibbonWidth:NiagaraFloat=&MaskFloat(Mask Source=0, Mask Parameter=?, Mask=0, Masked Float=90, Passthrough Float=-90)) {isSetParameters=true, module=SetVariables_#6}
  ]
  particleUpdateScript [
   &ParticleState()
   &ScaleColor(
    Scale Alpha=&MakeCustomFloatFromBool(
     Bool=$User.bIsPainted
     True Float=0.15
     False Float=&FloatFromCurve(FloatCurve=DataInterface() {propertyValues {Curve {keys [<at 0 0>,<at 0.02 0.1>,<at 0.95 0.1>,<at 1 0>]}}}, Scale Curve=0.2)
    )
    ScaleA=true
    ScaleRGB=true
   )
   &ScaleRibbonWidth(Ribbon Width Scale=&FloatFromCurve(
     FloatCurve=DataInterface() {propertyValues {Curve {keys [
         {interpMode=RCIM_Cubic, tangentMode=RCTM_User, time=0, value=0}
         {arriveTangent=0.5028, interpMode=RCIM_Cubic, leaveTangent=0.5028, tangentMode=RCTM_User, time=0.1, value=1.25}
         {arriveTangent=-0.5446, interpMode=RCIM_Cubic, leaveTangent=-0.5446, tangentMode=RCTM_User, time=0.8, value=1}
         {arriveTangent=-3.985, interpMode=RCIM_Cubic, leaveTangent=-3.985, tangentMode=RCTM_User, time=0.9, value=0}
        ]}}}
     Scale Curve=1.25
    ))
   /SetVariables(
    Particles.CustomWidth:NiagaraFloat=&Multiply_Float(
     A=$Particles.MainRibbonWidth
     B=&FloatFromCurve(
      FloatCurve=DataInterface() {propertyValues {Curve {keys [
          {interpMode=RCIM_Cubic, tangentMode=RCTM_User, time=0, value=0}
          {arriveTangent=0.5028, interpMode=RCIM_Cubic, leaveTangent=0.5028, tangentMode=RCTM_User, time=0.1, value=1.25}
          {arriveTangent=-0.5446, interpMode=RCIM_Cubic, leaveTangent=-0.5446, tangentMode=RCTM_User, time=0.8, value=1}
          {arriveTangent=-3.985, interpMode=RCIM_Cubic, leaveTangent=-3.985, tangentMode=RCTM_User, time=0.9, value=0}
         ]}}}
      Scale Curve=1.25
     )
    )
    Particles.SizeScaleTwo:NiagaraFloat=&Multiply_Float(
     A=$Particles.SizeTwo
     B=&FloatFromCurve(FloatCurve=DataInterface() {propertyValues {Curve {keys [
          {arriveTangent=-0.0009, interpMode=RCIM_Cubic, leaveTangent=-0.0009, tangentMode=RCTM_User, time=0, value=0}
          {arriveTangent=0.6917, interpMode=RCIM_Cubic, leaveTangent=0.6917, tangentMode=RCTM_User, time=0.5, value=1}
          {arriveTangent=-2.5286, interpMode=RCIM_Cubic, leaveTangent=-2.5286, tangentMode=RCTM_User, time=1, value=0}
         ]}}})
    )
   ) {isSetParameters=true, module=SetVariables_#7}
   &SolveForcesAndVelocity(Clamp Velocity=false, Limit Acceleration=false)
   &VC_UpdateRibbon_v1(Enable Skeletal Position=true)
   /SetVariables(Particles.Position=&AddVectorToPosition(
     Position=$Particles.Position
     Vector=&TransformVector(
      Vector=&MakeVector(
       X=&Sine(
        Normalized Angle=&Add_Float(
         A=&Multiply_Float(A=$Particles.NormalizedAge, B=1)
         B=&Multiply_Float(A=$Emitter.LoopedAge, B=0)
        )
        Period=$Emitter.TimeOffset
        Scale=&FloatFromCurve(
         FloatCurve=DataInterface() {propertyValues {Curve {keys [{arriveTangent=5.262, interpMode=RCIM_Cubic, leaveTangent=5.262, tangentMode=RCTM_User, time=0, value=0}, <at 0.5 1>]}}}
         Scale Curve=0
        )
        Bias=0
       )
       Y=&Sine(
        Normalized Angle=&Add_Float(
         A=&Multiply_Float(
          A=$Particles.Age
          B=&SelectFloatFromArray(Direct Array Index=$Particles.SpawnGroup, Float Selection Array=DataInterface() {propertyValues {FloatData [-1,1]}})
         )
         B=&Multiply_Float(A=$Emitter.LoopedAge, B=0)
        )
        Period=$Emitter.TimeOffset
        Scale=&FloatFromCurve(
         FloatCurve=DataInterface() {propertyValues {Curve {keys [{arriveTangent=0.7811, interpMode=RCIM_Cubic, leaveTangent=0.7811, tangentMode=RCTM_User, time=0, value=0.85}, <at 0.5 1>]}}}
         Scale Curve=4
        )
        Bias=0
       )
       Z=&Sine(
        Normalized Angle=&Add_Float(
         A=&Multiply_Float(
          A=$Particles.Age
          B=&SelectFloatFromArray(Direct Array Index=$Particles.SpawnGroup, Float Selection Array=DataInterface() {propertyValues {FloatData [1,1]}})
         )
         B=&Multiply_Float(A=$Emitter.LoopedAge, B=0)
        )
        Period=$Emitter.AnotherOffset
        Scale=&FloatFromCurve(
         FloatCurve=DataInterface() {propertyValues {Curve {keys [{arriveTangent=0.7811, interpMode=RCIM_Cubic, leaveTangent=0.7811, tangentMode=RCTM_User, time=0, value=0.85}, <at 0.5 1>]}}}
         Scale Curve=5
        )
        Bias=0
       )
      )
      Destination Space=::Simulation
      Apply Scale=true
     )
    )) {isSetParameters=true, module=SetVariables_#8}
  ]
  rendererClasses [&NiagaraRibbonRendererProperties]
  renderers [
   {rendererClass=&NiagaraRibbonRendererProperties}
   {rendererClass=&NiagaraRibbonRendererProperties, rendererIndex=1}
  ]
 }
 NE_VC_Boost_Ribbon_Parent005 {
  emitterSpawnScript [/SetVariables(
    Emitter.IsPainted:NiagaraFloat=&MakeCustomFloatFromBool(Bool=$User.bIsPainted, True Float=2, False Float=-0.512)
    Emitter.MirrorOffset:NiagaraFloat=&MakeCustomFloatFromBool(Bool=$User.bIsPainted, True Float=4, False Float=2)
   ) {isSetParameters=true, module=SetVariables_#2}]
  emitterUpdateScript [
   &EmitterState(Life Cycle Mode=::NewEnumerator1, Loop Duration=1, UseLoopDelay=false)
   &VC_FilteredSocketCount()
   &VC_SpawnRibbon_v1(Number of Ribbons=$Emitter.PortCount, SpawnRate=30)
   /SetVariables(
    Emitter.AnotherOffset:NiagaraFloat=&Power(a=HlslExpression() {hlslExpression="(fmod(Engine.Time * 0.35, 0.2) + 0.0) * 3"}, b=1)
    Emitter.TimeLerp:NiagaraFloat=&Power(a=HlslExpression() {hlslExpression="(fmod(Engine.Time * 0.325, 0.4) + 0.15) * 3"}, b=2)
    Emitter.TimeOffset:NiagaraFloat=&Power(a=HlslExpression() {hlslExpression="fmod(Engine.Time * 0.5, 0.3) + 0.15 * 3"}, b=2)
   ) {isSetParameters=true, module=SetVariables_#3}
   /SetVariables(
    Emitter.LerpChannel:Vector3f=&Lerp_Vector(
     A=<1 0 0>
     B=<0 1 0>
     Alpha=&Ceil_Float(Float=&Sine(Normalized Angle=$Emitter.Age, Period=$Emitter.TimeLerp, Scale=1, Bias=0))
    )
    Emitter.OffsetX:NiagaraFloat=&Multiply_Float(
     A=&Ceil_Float(Float=&Sine(Normalized Angle=$Emitter.Age, Period=$Emitter.TimeOffset, Scale=1, Bias=0))
     B=0.25
    )
   ) {isSetParameters=true, module=SetVariables_#4}
  ]
  particleSpawnScript [
   &VC_PaintedVariantsOverride() {enabled=false}
   /SetVariables(
    Particles.ColorTwo:LinearColor=<rgba 6 0 5.2125 1>
    Particles.SizeTwo:NiagaraFloat=&RandomRangeFloat(Minimum=60, Maximum=60)
   ) {isSetParameters=true, module=SetVariables_#5}
   &VC_InitializeRibbon_v1(
    Color=&MakeCustomLinearColorFromBool(Bool=$User.bIsPainted, True Color=$User.BelowTrailColor, False Color=<rgba 24 3.0925 20.1246 1>)
    Custom Offset=true
    Enable Skeletal Position=true
    Lifetime=0.46
    Offset Amount=1
    Ribbon Width:NiagaraFloat=&RandomRangeFloat(Minimum=90, Maximum=90)
   )
   &AddVelocity(Velocity=<-2500 0 50>) {module=AddVelocity001}
   &ShapeLocation(MaskSpawnGroup=false, Override Seed=false, Sphere Radius=7, UseOwnerScale=true) {enabled=false}
  ]
  particleUpdateScript [
   &ParticleState()
   &ScaleColor(
    Scale Alpha=&MakeCustomFloatFromBool(
     Bool=$User.bIsPainted
     True Float=0.15
     False Float=&FloatFromCurve(FloatCurve=DataInterface() {propertyValues {Curve {keys [<at 0 0>,<at 0.02 0.1>,<at 0.95 0.1>,<at 1 0>]}}}, Scale Curve=0.2)
    )
    ScaleA=true
    ScaleRGB=true
   )
   /SetVariables(Particles.SizeScaleTwo:NiagaraFloat=&Multiply_Float(
     A=$Particles.SizeTwo
     B=&FloatFromCurve(FloatCurve=DataInterface() {propertyValues {Curve {keys [<at 0 1>,<at 1 1>]}}})
    )) {isSetParameters=true, module=SetVariables_#7}
   &SolveForcesAndVelocity(Clamp Velocity=false, Limit Acceleration=false)
   &VC_UpdateRibbon_v1(Enable Skeletal Position=true)
   /SetVariables(Particles.CustomPosition:NiagaraPosition=&AddVectorToPosition(
     Position=$Particles.Position
     Vector=&TransformVector(
      Vector=&MakeVector(
       X=0
       Y=&Sine(
        Normalized Angle=&Multiply_Float(
         A=1
         B=&Subtract_Float(
          A=&Multiply_Float(A=2, B=&Subtract_Float(A=$Engine.Time, B=0.5))
          B=$Particles.NormalizedAge
         )
        )
        Period=&SelectFloatFromArray(
         Direct Array Index=&SpawnGroup()
         Float Selection Array=DataInterface() {propertyValues {FloatData [0.7,0.765]}}
        )
        Scale=&FloatFromCurve(FloatCurve=DataInterface() {propertyValues {Curve {keys [<at 0 60 RCIM_Cubic>,<at 0.86 0 RCIM_Cubic>]}}})
        Bias=0
       )
       Z=&Sine(
        Normalized Angle=&Multiply_Float(
         A=1
         B=&Subtract_Float(
          A=&Multiply_Float(A=2, B=&Subtract_Float(A=$Engine.Time, B=0.5))
          B=$Particles.NormalizedAge
         )
        )
        Period=&SelectFloatFromArray(
         Direct Array Index=&SpawnGroup()
         Float Selection Array=DataInterface() {propertyValues {FloatData [0.3,0.365]}}
        )
        Scale=&FloatFromCurve(FloatCurve=DataInterface() {propertyValues {Curve {keys [<at 0.1 2>,<at 1 48>]}}})
        Bias=0
       )
      )
      Destination Space=::Simulation
      Apply Scale=true
     )
    )) {isSetParameters=true, module=SetVariables_#10}
  ]
  rendererClasses [&NiagaraRibbonRendererProperties]
  renderers [
   {rendererClass=&NiagaraRibbonRendererProperties}
   {rendererClass=&NiagaraRibbonRendererProperties, rendererIndex=1}
   {rendererClass=&NiagaraRibbonRendererProperties, rendererIndex=2}
  ]
 }
 NE_VC_Boost_Ribbon_Parent007 {
  bEnabled=false
  emitterUpdateScript [
   &EmitterState()
   &VC_FilteredSocketCount()
   &VC_RateQualityHandler()
   &VC_SpawnRibbon_v1(Number of Ribbons=$Emitter.PortCount, SpawnRate=$Emitter.Spawn Rate)
   /SetVariables(Emitter.TimeOffset:NiagaraFloat=&Power(a=HlslExpression() {hlslExpression="(fmod(Engine.Time * 0.35, 0.2) + 0.0) * 3"}, b=2)) {isSetParameters=true, module=SetVariables_#11}
  ]
  particleSpawnScript [
   &VC_PaintedVariantsOverride() {enabled=false}
   &VC_InitializeRibbon_v1(
    Custom Offset=true
    Enable Skeletal Position=true
    Lifetime=0.7
    Local Vector Offset=<20 0 0>
    Offset Amount=1
    Ribbon Width:NiagaraFloat=&RandomRangeFloat(Minimum=64, Maximum=64)
   )
   &AddVelocity(Velocity=<-2000 0 0>) {module=AddVelocity001}
   &ShapeLocation(MaskSpawnGroup=false, Override Seed=false, Sphere Radius=2, UseOwnerScale=true) {enabled=false}
  ]
  particleUpdateScript [
   &ParticleState()
   &ScaleRibbonWidth(Ribbon Width Scale=&FloatFromCurve(FloatCurve=DataInterface() {propertyValues {Curve {keys [
         {arriveTangent=1.0826, interpMode=RCIM_Cubic, leaveTangent=1.0826, tangentMode=RCTM_User, time=0, value=0.5}
         {arriveTangent=0.0088, interpMode=RCIM_Cubic, leaveTangent=0.0088, tangentMode=RCTM_User, time=0.25, value=0.7}
         {arriveTangent=-1.7516, interpMode=RCIM_Cubic, leaveTangent=-1.7516, tangentMode=RCTM_User, time=0.8, value=0}
        ]}}}))
   &SolveForcesAndVelocity(Clamp Velocity=false, Limit Acceleration=false)
   &VC_UpdateRibbon_v1(Enable Skeletal Position=true)
   /SetVariables(Particles.Position=&AddVectorToPosition(
     Position=$Particles.Position
     Vector=&TransformVector(
      Vector=&MakeVector(
       X=&Sine(
        Normalized Angle=&Add_Float(
         A=&Multiply_Float(A=$Particles.NormalizedAge, B=1)
         B=&Multiply_Float(A=$Emitter.LoopedAge, B=0)
        )
        Period=$Emitter.TimeOffset
        Scale=&FloatFromCurve(
         FloatCurve=DataInterface() {propertyValues {Curve {keys [{arriveTangent=5.262, interpMode=RCIM_Cubic, leaveTangent=5.262, tangentMode=RCTM_User, time=0, value=0}, <at 0.5 1>]}}}
         Scale Curve=0
        )
        Bias=0
       )
       Y=&Sine(
        Normalized Angle=&Add_Float(
         A=&Multiply_Float(
          A=$Particles.Age
          B=&SelectFloatFromArray(Direct Array Index=$Particles.SpawnGroup, Float Selection Array=DataInterface() {propertyValues {FloatData [-1,1]}})
         )
         B=&Multiply_Float(A=$Emitter.LoopedAge, B=0)
        )
        Period=$Emitter.TimeOffset
        Scale=&FloatFromCurve(
         FloatCurve=DataInterface() {propertyValues {Curve {keys [{arriveTangent=0.7811, interpMode=RCIM_Cubic, leaveTangent=0.7811, tangentMode=RCTM_User, time=0, value=0.85}, <at 0.5 1>]}}}
         Scale Curve=4
        )
        Bias=0
       )
       Z=&Sine(
        Normalized Angle=&Add_Float(
         A=&Multiply_Float(
          A=$Particles.Age
          B=&SelectFloatFromArray(Direct Array Index=$Particles.SpawnGroup, Float Selection Array=DataInterface() {propertyValues {FloatData [-1,1]}})
         )
         B=&Multiply_Float(A=$Emitter.LoopedAge, B=0)
        )
        Period=$Emitter.TimeOffset
        Scale=&FloatFromCurve(
         FloatCurve=DataInterface() {propertyValues {Curve {keys [{arriveTangent=0.7811, interpMode=RCIM_Cubic, leaveTangent=0.7811, tangentMode=RCTM_User, time=0, value=0.85}, <at 0.5 1>]}}}
         Scale Curve=5
        )
        Bias=0
       )
      )
      Destination Space=::Simulation
      Apply Scale=true
     )
    )) {isSetParameters=true, module=SetVariables_#12}
  ]
  rendererClasses [&NiagaraRibbonRendererProperties]
  renderers [{rendererClass=&NiagaraRibbonRendererProperties}]
 }
 NE_VC_Boost_Sprite_Parent003 {
  emitterUpdateScript [
   &EmitterState()
   &VC_FilteredSocketCount()
   &SpawnRate(SpawnRate=10, Use Spawn Probability=false)
   &SpawnRate(SpawnGroup=1, SpawnRate=10, Use Spawn Probability=false) {module=SpawnRate001}
  ]
  particleSpawnScript [
   /SetVariables(Particles.VisibilityTag=&SpawnGroup()) {isSetParameters=true}
   &VC_PaintedVariantsOverride() {enabled=false}
   &InitializeParticle(
    Color=&MakeCustomLinearColorFromBool(Bool=$User.bIsPainted, True Color=$User.PaintedColor, False Color=<rgba 15 2.5 10 1>)
    Lifetime=&RandomRangeFloat(Minimum=0.5, Maximum=0.75)
    Sprite Rotation Mode=::NewEnumerator2
    Sprite Size Mode=::NewEnumerator3
    Uniform Sprite Size=&SelectFloatFromArray(Direct Array Index=&SpawnGroup(), Float Selection Array=DataInterface() {propertyValues {FloatData [12,8]}})
    UsePositionOffset=false
    Write Material Random=true
   )
   &PartitionParticles(Number Of Partitions=$Emitter.PortCount, Partition Mode:ENiagara_PartitionMode=::NewEnumerator5)
   &SkeletalMeshLocation(
    Bone / Socket Index=$Particles.PartitionParticles.Partition
    Bone / Socket Sampling Mode=::NewEnumerator5
    MaskSpawnGroup=false
    Mesh Sampling Type=::NewEnumerator6
    Skeletal Mesh=$System.SkeletalMesh
   )
   &AddVelocity(Velocity=<-750 0 0>)
   &ShapeLocation(MaskSpawnGroup=true, Override Seed=false, Sphere Radius=30, UseOwnerScale=true)
   &ShapeLocation(MaskSpawnGroup=true, Override Seed=false, Spawn Group Mask=1, Sphere Radius=25, UseOwnerScale=true) {module=ShapeLocation001}
  ]
  particleUpdateScript [
   &ParticleState()
   &ScaleColor(
    Scale Alpha=&FloatFromCurve(FloatCurve=DataInterface() {propertyValues {Curve {keys [<at 0 0>,<at 0.05 1>,<at 0.75 1>,<at 1 0>]}}})
    ScaleA=true
    ScaleRGB=true
   )
   &ScaleSpriteSize(Uniform Curve Sprite Scale=DataInterface() {propertyValues {Curve {keys [<at 0 0 RCIM_Cubic>, {arriveTangent=2, time=0.1, value=1.5}, <at 0.5 1.5>, {tangentMode=RCTM_User, time=1, value=0}]}}})
   &CurlNoiseForce(Bias Noise=false, Noise Frequency=20, Noise Strength=12000, Pan Noise=false, Pan Noise Field=<0.23 0.84 0.47>)
   &Drag(Drag=0.5, Use Linear Drag=true, Use Rotational Drag=false)
   &ScaleVelocity(
    Coordinate Space:ENiagaraCoordinateSpace=::Simulation
    Velocity Scale=&VectorFromCurve(VectorCurve=DataInterface() {propertyValues {
       XCurve {keys [<at 0.5 1 RCIM_Cubic>,<at 1 0 RCIM_Cubic>]}
       YCurve {keys [<at 0.5 1 RCIM_Cubic>,<at 1 0 RCIM_Cubic>]}
       ZCurve {keys [<at 0.5 1 RCIM_Cubic>,<at 1 0 RCIM_Cubic>]}
      }})
   )
   &SolveForcesAndVelocity(Clamp Velocity=false, Limit Acceleration=false)
   &JitterPosition(Jitter Amount=1) {enabled=false}
  ]
  rendererClasses [&NiagaraSpriteRendererProperties]
  renderers [
   {rendererClass=&NiagaraSpriteRendererProperties}
   {rendererClass=&NiagaraSpriteRendererProperties, rendererIndex=1}
   {rendererClass=&NiagaraSpriteRendererProperties, rendererIndex=2}
  ]
 }
}
summary {
 systemName=NS_Boost_Electrobeam_fg
 emitters [
  {emitterName=NE_VC_Boost_Sprite_Parent003, rendererClasses [&NiagaraSpriteRendererProperties]}
  {bEnabled=false, emitterName=NE_VC_Boost_Ribbon_Parent004, rendererClasses [&NiagaraRibbonRendererProperties]}
  {bEnabled=false, emitterName=NE_VC_Boost_Ribbon_Parent007, rendererClasses [&NiagaraRibbonRendererProperties]}
  {bEnabled=false, emitterName=NE_VC_Boost_Ribbon_Parent, rendererClasses [&NiagaraRibbonRendererProperties]}
  {bEnabled=false, emitterName=NE_VC_Boost_Ribbon_Parent002, rendererClasses [&NiagaraRibbonRendererProperties]}
  {emitterName=NE_VC_Boost_Ribbon_Parent001, rendererClasses [&NiagaraRibbonRendererProperties]}
  {emitterName=NE_VC_Boost_Ribbon_Parent003, rendererClasses [&NiagaraRibbonRendererProperties]}
  {emitterName=NE_VC_Boost_Ribbon_Parent005, rendererClasses [&NiagaraRibbonRendererProperties]}
 ]
 userVariables [
  User.BelowTrailColor {type=LinearColor, description="", defaultValue {struct=LinearColor, value=<rgba 1 1 1 1>}}
  User.bIsFrontendPreview {type=NiagaraBool}
  User.bIsPainted {type=NiagaraBool}
  User.ColorVariantID {type=NiagaraInt32}
  User.PaintedColor {type=LinearColor, description="", defaultValue {struct=LinearColor, value=<rgba 1 1 1 1>}}
  User.PaintedScalar {type=NiagaraFloat, defaultValue=1, description=""}
  User.PortCount {type=NiagaraInt32, defaultValue=1}
  User.SizeScaling {type=NiagaraFloat, defaultValue=8, description=""}
  User.SubTrailColor {type=LinearColor, description="", defaultValue {struct=LinearColor, value=<rgba 1 1 1 1>}}
  User.TopTrailColor {type=LinearColor, description="", defaultValue {struct=LinearColor, value=<rgba 1 1 1 1>}}
  User.TrackTrailColor {type=LinearColor, description="", defaultValue {struct=LinearColor, value=<rgba 1 1 1 1>}}
 ]
}
